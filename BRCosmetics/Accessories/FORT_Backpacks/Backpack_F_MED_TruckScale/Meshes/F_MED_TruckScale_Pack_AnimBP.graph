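# Animation Blueprint for the TruckScale backpack cosmetic, making the pack follow the spine of the mesh it is attached to.
# Each update, the event graph reads pelvis and five spine bone world transforms from the attach parent into variables.
# The AnimGraph writes them onto the pack's matching bones in world space.
# Constraints then drive top, middle and bottom block base bones from weighted blends of those spine bones.
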AnimBlueprint /BRCosmetics/Accessories/FORT_Backpacks/Backpack_F_MED_TruckScale/Meshes/F_MED_TruckScale_Pack_AnimBP
target_skeleton=&/F_MED_TruckScale_Pack_Skeleton
graphs [
 AnimGraph {nodes [
   Input Pose()
   n1 = "Transform (Modify) Bone - Bone: pelvis"(
    ComponentPose=Local To Component(LocalPose=Local Space Ref Pose())
    Alpha=1
    Translation=Get pelvis_01_tran() -> (pelvis_01_tran:Vector)
    Rotation=Get pelvis_01_rot() -> (pelvis_01_rot:Rotator)
    Scale="1,1,1"
   ) {
    rotationMode=BMM_Replace
    rotationSpace=BCS_WorldSpace
    scaleMode=BMM_Replace
    scaleSpace=BCS_WorldSpace
    translationMode=BMM_Replace
    translationSpace=BCS_WorldSpace
    alphaBoolBlend {blendOption=Linear}
    alphaScaleBiasClamp {inRange {max=0, min=-40}, outRange {max=0, min=1}}
    boneToModify {boneName=pelvis}
   }
   n2 = "Transform (Modify) Bone - Bone: spine_01"(
    ComponentPose=@n1
    Alpha=1
    Translation=Get spine_01_tran() -> (spine_01_tran:Vector)
    Rotation=Get spine_01_rot() -> (spine_01_rot:Rotator)
    Scale="1,1,1"
   ) {
    rotationMode=BMM_Replace
    rotationSpace=BCS_WorldSpace
    scaleMode=BMM_Replace
    scaleSpace=BCS_WorldSpace
    translationMode=BMM_Replace
    translationSpace=BCS_WorldSpace
    alphaBoolBlend {blendOption=Linear}
    alphaScaleBiasClamp {inRange {max=0, min=-40}, outRange {max=0, min=1}}
    boneToModify {boneName=spine_01}
   }
   n3 = "Transform (Modify) Bone - Bone: spine_02"(
    ComponentPose=@n2
    Alpha=1
    Translation=Get spine_02_tran() -> (spine_02_tran:Vector)
    Rotation=Get spine_02_rot() -> (spine_02_rot:Rotator)
    Scale="1,1,1"
   ) {
    rotationMode=BMM_Replace
    rotationSpace=BCS_WorldSpace
    scaleMode=BMM_Replace
    scaleSpace=BCS_WorldSpace
    translationMode=BMM_Replace
    translationSpace=BCS_WorldSpace
    alphaBoolBlend {blendOption=Linear}
    alphaScaleBiasClamp {inRange {max=0, min=-40}, outRange {max=0, min=1}}
    boneToModify {boneName=spine_02}
   }
   n4 = "Transform (Modify) Bone - Bone: spine_03"(
    ComponentPose=@n3
    Alpha=1
    Translation=Get spine_03_tran() -> (spine_03_tran:Vector)
    Rotation=Get spine_03_rot() -> (spine_03_rot:Rotator)
    Scale="1,1,1"
   ) {
    rotationMode=BMM_Replace
    rotationSpace=BCS_WorldSpace
    scaleMode=BMM_Replace
    scaleSpace=BCS_WorldSpace
    translationMode=BMM_Replace
    translationSpace=BCS_WorldSpace
    alphaBoolBlend {blendOption=Linear}
    alphaScaleBiasClamp {inRange {max=0, min=-40}, outRange {max=0, min=1}}
    boneToModify {boneName=spine_03}
   }
   n5 = "Transform (Modify) Bone - Bone: spine_04"(
    ComponentPose=@n4
    Alpha=1
    Translation=Get spine_04_tran() -> (spine_04_tran:Vector)
    Rotation=Get spine_04_rot() -> (spine_04_rot:Rotator)
    Scale="1,1,1"
   ) {
    rotationMode=BMM_Replace
    rotationSpace=BCS_WorldSpace
    scaleMode=BMM_Replace
    scaleSpace=BCS_WorldSpace
    translationMode=BMM_Replace
    translationSpace=BCS_WorldSpace
    alphaBoolBlend {blendOption=Linear}
    alphaScaleBiasClamp {inRange {max=0, min=-40}, outRange {max=0, min=1}}
    boneToModify {boneName=spine_04}
   }
   n6 = "Transform (Modify) Bone - Bone: spine_05"(
    ComponentPose=@n5
    Alpha=1
    Translation=Get spine_05_tran() -> (spine_05_tran:Vector)
    Rotation=Get spine_05_rot() -> (spine_05_rot:Rotator)
    Scale="1,1,1"
   ) {
    rotationMode=BMM_Replace
    rotationSpace=BCS_WorldSpace
    scaleMode=BMM_Replace
    scaleSpace=BCS_WorldSpace
    translationMode=BMM_Replace
    translationSpace=BCS_WorldSpace
    alphaBoolBlend {blendOption=Linear}
    alphaScaleBiasClamp {inRange {max=0, min=-40}, outRange {max=0, min=1}}
    boneToModify {boneName=spine_05}
   }
   n7 = "Constraint - block_top_BASE (P)"(ComponentPose=@n6, Alpha=1, ConstraintWeights_0=1) {
    alphaBoolBlend {blendOption=Linear}
    boneToModify {boneName=block_top_BASE}
    constraintSetup [{targetBone {boneName=spine_05}}]
    constraintWeights [1]
   }
   n8 = "Constraint - block_mid_BASE (PP)"!(ComponentPose=@n7, Alpha=1, ConstraintWeights_0=0.9, ConstraintWeights_1:float=0.1) -> (Pose) {
    alphaBoolBlend {blendOption=Linear}
    boneToModify {boneName=block_mid_BASE}
    constraintSetup [{targetBone {boneName=spine_05}}, {targetBone {boneName=spine_04}}]
    constraintWeights [1,1]
   }
   n9 = "Constraint - block_bot_BASE (PPPPP)"!(
    ComponentPose=@n8
    Alpha=1
    ConstraintWeights_0=0.544
    ConstraintWeights_1:float=0.353
    ConstraintWeights_2:float=0.047
    ConstraintWeights_3:float=0.043
    ConstraintWeights_4:float=0.013
   ) -> (Pose) {
    alphaBoolBlend {blendOption=Linear}
    boneToModify {boneName=block_bot_BASE}
    constraintSetup [
     {targetBone {boneName=spine_05}}
     {targetBone {boneName=spine_04}}
     {targetBone {boneName=spine_03}}
     {targetBone {boneName=spine_02}}
     {targetBone {boneName=spine_01}}
    ]
    constraintWeights [1,1,1,1,1]
   }
   Output Pose(Result=Component To Local(ComponentPose=@n9)) {layerGroup=None}
  ]}
 EventGraph {nodes [
   Event BlueprintUpdateAnimation()
   n1 = GetPartSkeletalMeshComponent()
   > Is Valid(InputObject=@n1)
   n2 = GetAttachParent!(self:Scene Component Object Reference=@n1) -> (ReturnValue:Scene Component Object Reference)
   > Is Valid(InputObject=@n2)
   > n3 = Sequence!(execute) -> (then_0, then_1, then_2:Exec, then_3:Exec, then_4:Exec, then_5:Exec)
   n4 = BreakTransform(InTransform=GetSocketTransform(self=@n2, InSocketName=spine_01, TransformSpace=RTS_World))
   > Set spine_01_tran(spine_01_tran:Vector=@n4) -> (Output_Get:Vector)
   > Set spine_01_rot(spine_01_rot:Rotator=@n4.Rotation) -> (Output_Get:Rotator)
   n5 = BreakTransform(InTransform=GetSocketTransform(self=@n2, InSocketName=spine_02, TransformSpace=RTS_World))
   Set spine_02_tran(execute=@n3.then_1, spine_02_tran:Vector=@n5) -> (Output_Get:Vector)
   > Set spine_02_rot(spine_02_rot:Rotator=@n5.Rotation) -> (Output_Get:Rotator)
   n6 = BreakTransform(InTransform=GetSocketTransform(self=@n2, InSocketName=spine_03, TransformSpace=RTS_World))
   Set spine_03_tran(execute=@n3.then_2, spine_03_tran:Vector=@n6) -> (Output_Get:Vector)
   > Set spine_03_rot(spine_03_rot:Rotator=@n6.Rotation) -> (Output_Get:Rotator)
   n7 = BreakTransform(InTransform=GetSocketTransform(self=@n2, InSocketName=spine_04, TransformSpace=RTS_World))
   Set spine_04_tran(execute=@n3.then_3, spine_04_tran:Vector=@n7) -> (Output_Get:Vector)
   > Set spine_04_rot(spine_04_rot:Rotator=@n7.Rotation) -> (Output_Get:Rotator)
   n8 = BreakTransform(InTransform=GetSocketTransform(self=@n2, InSocketName=spine_05, TransformSpace=RTS_World))
   Set spine_05_tran(execute=@n3.then_4, spine_05_tran:Vector=@n8) -> (Output_Get:Vector)
   > Set spine_05_rot(spine_05_rot:Rotator=@n8.Rotation) -> (Output_Get:Rotator)
   n9 = BreakTransform(InTransform=GetSocketTransform(self=@n2, InSocketName=pelvis, TransformSpace=RTS_World))
   Set pelvis_01_tran(execute=@n3.then_5, pelvis_01_tran:Vector=@n9) -> (Output_Get:Vector)
   > Set pelvis_01_rot(pelvis_01_rot:Rotator=@n9.Rotation) -> (Output_Get:Rotator)
   TryGetPawnOwner()
  ]}
]
variables [
 pelvis_01_rot {type=Rotator}
 pelvis_01_tran {type=Vector}
 spine_01_rot {type=Rotator}
 spine_01_tran {type=Vector}
 spine_02_rot {type=Rotator}
 spine_02_tran {type=Vector}
 spine_03_rot {type=Rotator}
 spine_03_tran {type=Vector}
 spine_04_rot {type=Rotator}
 spine_04_tran {type=Vector}
 spine_05_rot {type=Rotator}
 spine_05_tran {type=Vector}
]
